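# Animation Blueprint for the IvyStub pickaxe that wobbles the rock point bones on the axe.
# Each update, the EventGraph scales one sine of real time into per-stone offset vectors and seven rotators.
# The AnimGraph chains fifteen additive bone-space Modify Bone nodes over the input pose: seven read the oscillating rotators, eight use fixed roll offsets.
# The offset vectors are computed but unused, as every translation input stays zero.
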AnimBlueprint /BRCosmetics/Weapons/FORT_Melee/Pickaxe_IvyStub/Meshes/IvyStub_Axe_AnimBP
target_skeleton=&/IvyStub_Axe_Skeleton
graphs [
 AnimGraph {nodes [
   AnimGraphNode_SubInput_0 = Input Pose()
   n1 = "Transform (Modify) Bone - Bone: base_rockaxePoint_01_c"(
    ComponentPose=Local To Component(LocalPose=@AnimGraphNode_SubInput_0)
    Alpha=1
    Translation="0,0,0"
    Rotation=Get StoneRot_01() -> (StoneRot_01:Rotator)
    Scale="1,1,1"
   ) {
    rotationMode=BMM_Additive
    rotationSpace=BCS_BoneSpace
    translationMode=BMM_Additive
    translationSpace=BCS_BoneSpace
    alphaBoolBlend {blendOption=Linear}
    boneToModify {boneName=base_rockaxePoint_01_c}
   }
   n2 = "Transform (Modify) Bone - Bone: base_rockaxePoint_02_c"(ComponentPose=@n1, Alpha=1, Translation="0,0,0", Rotation="0,0,7", Scale="1,1,1") {
    rotationMode=BMM_Additive
    rotationSpace=BCS_BoneSpace
    translationMode=BMM_Additive
    translationSpace=BCS_BoneSpace
    alphaBoolBlend {blendOption=Linear}
    boneToModify {boneName=base_rockaxePoint_02_c}
    rotation {roll=7}
   }
   n3 = "Transform (Modify) Bone - Bone: base_rockaxePoint_03_c"(ComponentPose=@n2, Alpha=1, Translation="0,0,0", Rotation="0,0,4", Scale="1,1,1") {
    rotationMode=BMM_Additive
    rotationSpace=BCS_BoneSpace
    translationMode=BMM_Additive
    translationSpace=BCS_BoneSpace
    alphaBoolBlend {blendOption=Linear}
    boneToModify {boneName=base_rockaxePoint_03_c}
    rotation {roll=4}
   }
   n4 = "Transform (Modify) Bone - Bone: base_rockaxePoint_04_c"(ComponentPose=@n3, Alpha=1, Translation="0,0,0", Rotation=Get StoneRot_04() -> (StoneRot_04:Rotator), Scale="1,1,1") {
    rotationMode=BMM_Additive
    rotationSpace=BCS_BoneSpace
    translationMode=BMM_Additive
    translationSpace=BCS_BoneSpace
    alphaBoolBlend {blendOption=Linear}
    boneToModify {boneName=base_rockaxePoint_04_c}
   }
   n5 = "Transform (Modify) Bone - Bone: base_rockaxePoint_05_c"(ComponentPose=@n4, Alpha=1, Translation="0,0,0", Rotation=Get StoneRot_05() -> (StoneRot_05:Rotator), Scale="1,1,1") {
    rotationMode=BMM_Additive
    rotationSpace=BCS_BoneSpace
    translationMode=BMM_Additive
    translationSpace=BCS_BoneSpace
    alphaBoolBlend {blendOption=Linear}
    boneToModify {boneName=base_rockaxePoint_05_c}
   }
   n6 = "Transform (Modify) Bone - Bone: base_rockaxePoint_06_c"(ComponentPose=@n5, Alpha=1, Translation="0,0,0", Rotation=Get StoneRot_06() -> (StoneRot_06:Rotator), Scale="1,1,1") {
    rotationMode=BMM_Additive
    rotationSpace=BCS_BoneSpace
    translationMode=BMM_Additive
    translationSpace=BCS_BoneSpace
    alphaBoolBlend {blendOption=Linear}
    boneToModify {boneName=base_rockaxePoint_06_c}
   }
   n7 = "Transform (Modify) Bone - Bone: base_rockaxePoint_07_c"(ComponentPose=@n6, Alpha=1, Translation="0,0,0", Rotation="0,0,4", Scale="1,1,1") {
    rotationMode=BMM_Additive
    rotationSpace=BCS_BoneSpace
    translationMode=BMM_Additive
    translationSpace=BCS_BoneSpace
    alphaBoolBlend {blendOption=Linear}
    boneToModify {boneName=base_rockaxePoint_07_c}
    rotation {roll=4}
   }
   n8 = "Transform (Modify) Bone - Bone: base_rockaxePoint_08_c"(ComponentPose=@n7, Alpha=1, Translation="0,0,0", Rotation="0,0,5", Scale="1,1,1") {
    rotationMode=BMM_Additive
    rotationSpace=BCS_BoneSpace
    translationMode=BMM_Additive
    translationSpace=BCS_BoneSpace
    alphaBoolBlend {blendOption=Linear}
    boneToModify {boneName=base_rockaxePoint_08_c}
    rotation {roll=5}
   }
   n9 = "Transform (Modify) Bone - Bone: base_rockaxePoint_09_c"(ComponentPose=@n8, Alpha=1, Translation="0,0,0", Rotation="0,0,6", Scale="1,1,1") {
    rotationMode=BMM_Additive
    rotationSpace=BCS_BoneSpace
    translationMode=BMM_Additive
    translationSpace=BCS_BoneSpace
    alphaBoolBlend {blendOption=Linear}
    boneToModify {boneName=base_rockaxePoint_09_c}
    rotation {roll=6}
   }
   n10 = "Transform (Modify) Bone - Bone: base_rockaxePoint_10_c"(ComponentPose=@n9, Alpha=1, Translation="0,0,0", Rotation=Get StoneRot_10() -> (StoneRot_10:Rotator), Scale="1,1,1") {
    rotationMode=BMM_Additive
    rotationSpace=BCS_BoneSpace
    translationMode=BMM_Additive
    translationSpace=BCS_BoneSpace
    alphaBoolBlend {blendOption=Linear}
    boneToModify {boneName=base_rockaxePoint_10_c}
   }
   n11 = "Transform (Modify) Bone - Bone: base_rockaxePoint_11_c"(ComponentPose=@n10, Alpha=1, Translation="0,0,0", Rotation="0,0,6", Scale="1,1,1") {
    rotationMode=BMM_Additive
    rotationSpace=BCS_BoneSpace
    translationMode=BMM_Additive
    translationSpace=BCS_BoneSpace
    alphaBoolBlend {blendOption=Linear}
    boneToModify {boneName=base_rockaxePoint_11_c}
    rotation {roll=6}
   }
   n12 = "Transform (Modify) Bone - Bone: base_rockaxePoint_12_c"(ComponentPose=@n11, Alpha=1, Translation="0,0,0", Rotation=Get StoneRot_12() -> (StoneRot_12:Rotator), Scale="1,1,1") {
    rotationMode=BMM_Additive
    rotationSpace=BCS_BoneSpace
    translationMode=BMM_Additive
    translationSpace=BCS_BoneSpace
    alphaBoolBlend {blendOption=Linear}
    boneToModify {boneName=base_rockaxePoint_12_c}
   }
   n13 = "Transform (Modify) Bone - Bone: base_rockaxePoint_13_c"(ComponentPose=@n12, Alpha=1, Translation="0,0,0", Rotation="0,0,7", Scale="1,1,1") {
    rotationMode=BMM_Additive
    rotationSpace=BCS_BoneSpace
    translationMode=BMM_Additive
    translationSpace=BCS_BoneSpace
    alphaBoolBlend {blendOption=Linear}
    boneToModify {boneName=base_rockaxePoint_13_c}
    rotation {roll=7}
   }
   n14 = "Transform (Modify) Bone - Bone: base_rockaxePoint_14_c"(ComponentPose=@n13, Alpha=1, Translation="0,0,0", Rotation=Get StoneRot_14() -> (StoneRot_14:Rotator), Scale="1,1,1") {
    rotationMode=BMM_Additive
    rotationSpace=BCS_BoneSpace
    translationMode=BMM_Additive
    translationSpace=BCS_BoneSpace
    alphaBoolBlend {blendOption=Linear}
    boneToModify {boneName=base_rockaxePoint_14_c}
   }
   n15 = "Transform (Modify) Bone - Bone: base_rockaxePoint_15_c"(ComponentPose=@n14, Alpha=1, Translation="0,0,0", Rotation="0,0,4", Scale="1,1,1") {
    rotationMode=BMM_Additive
    rotationSpace=BCS_BoneSpace
    translationMode=BMM_Additive
    translationSpace=BCS_BoneSpace
    alphaBoolBlend {blendOption=Linear}
    boneToModify {boneName=base_rockaxePoint_15_c}
    rotation {roll=4}
   }
   Output Pose(Result=Component To Local(ComponentPose=@n15)) {layerGroup=None}
  ]}
 EventGraph {nodes [
   Event BlueprintUpdateAnimation()
   > Sequence!(execute) -> (then_0, then_1, then_2:Exec)
   n1 = "Sin (Radians)"(A=float * float(A=GetRealTimeSeconds(), B=1))
   > Set Stone_01_Float(Stone_01_Float:Vector=MakeVector(X=float * float(A=@n1, B=0.7) {is=title K2Node_PromotableOperator}, Y=float * float(A=@n1, B=0.7), Z=float * float(A=@n1, B=0.7))) -> (Output_Get:Vector)
   > Set StoneRot_01(StoneRot_01:Rotator=MakeRotator(Roll=float * float(A=@n1, B=3))) -> (Output_Get:Rotator)
   > Set Stone_02_Float(Stone_02_Float:Vector=MakeVector(X=float * float(A=@n1, B=0.7), Y=float * float(A=@n1, B=-0.7), Z=float * float(A=@n1, B=0.7))) -> (Output_Get:Vector)
   > Set Stone_03_Float(Stone_03_Float:Vector=MakeVector(X=float * float(A=@n1, B=-0.7), Y=float * float(A=@n1, B=0.8), Z=float * float(A=@n1, B=0.7))) -> (Output_Get:Vector)
   > Set Stone_04_Float(Stone_04_Float:Vector=MakeVector(X=float * float(A=@n1, B=-0.8), Y=float * float(A=@n1, B=0.9), Z=float * float(A=@n1, B=-0.8))) -> (Output_Get:Vector)
   > Set StoneRot_04(StoneRot_04:Rotator=MakeRotator(Pitch=float * float(A=@n1, B=10))) -> (Output_Get:Rotator)
   n2 = float * float(A=@n1, B=-0.2)
   n3 = float / float(A=@n1, B=1)
   > Set Stone_05_Float(Stone_05_Float:Vector=MakeVector(X=float * float(A=@n2, B=-0.8), Y=float * float(A=@n1, B=-0.8), Z=float * float(A=@n3, B=0.8))) -> (Output_Get:Vector)
   > Set StoneRot_05(StoneRot_05:Rotator=MakeRotator(Pitch=float * float(A=@n3, B=9))) -> (Output_Get:Rotator)
   > Set Stone_06_Float(Stone_06_Float:Vector=MakeVector(X=float * float(A=@n1, B=0.8), Y=float * float(A=@n1, B=0.8), Z=float * float(A=@n1, B=-0.8))) -> (Output_Get:Vector)
   > Set StoneRot_06(StoneRot_06:Rotator=MakeRotator(Pitch=float * float(A=@n1, B=8))) -> (Output_Get:Rotator)
   > Set Stone_07_Float(Stone_07_Float:Vector=MakeVector(X=float * float(A=@n1, B=-0.5), Y=float * float(A=@n1, B=1), Z=float * float(A=@n1, B=0.8))) -> (Output_Get:Vector)
   > Set Stone_08_Float(Stone_08_Float:Vector=MakeVector(X=float * float(A=@n1, B=-0.2), Y=float * float(A=@n1, B=1), Z=float * float(A=@n1, B=0.8))) -> (Output_Get:Vector)
   > Set Stone_09_Float(Stone_09_Float:Vector=MakeVector(X=float * float(A=@n1, B=0.4), Y=float * float(A=@n1, B=1.5), Z=float * float(A=@n1, B=0.6))) -> (Output_Get:Vector)
   > Set Stone_10_Float(Stone_10_Float:Vector=MakeVector(X=float * float(A=@n1, B=-0.8), Y=float * float(A=@n1, B=0.8), Z=float * float(A=@n1, B=0.8))) -> (Output_Get:Vector)
   > Set StoneRot_10(StoneRot_10:Rotator=MakeRotator(Pitch=float * float(A=@n1, B=5))) -> (Output_Get:Rotator)
   > Set Stone_11_Float(Stone_11_Float:Vector=MakeVector(X=float * float(A=@n1, B=-0.7), Y=float * float(A=@n1, B=-0.8), Z=float * float(A=@n1, B=-0.8))) -> (Output_Get:Vector)
   > Set Stone_12_Float(Stone_12_Float:Vector=MakeVector(X=float * float(A=@n1, B=-0.8), Y=float * float(A=@n1, B=0.8), Z=float * float(A=@n1, B=-0.8))) -> (Output_Get:Vector)
   > Set StoneRot_12(StoneRot_12:Rotator=MakeRotator(Pitch=float * float(A=@n1, B=-5))) -> (Output_Get:Rotator)
   > Set Stone_13_Float(Stone_13_Float:Vector=MakeVector(X=float * float(A=@n1, B=0.2), Y=float * float(A=@n1, B=1), Z=float * float(A=@n1, B=0.8))) -> (Output_Get:Vector)
   > Set Stone_14_Float(Stone_14_Float:Vector=MakeVector(X=float * float(A=@n1, B=-0.6), Y=float * float(A=@n1, B=-0.7), Z=float * float(A=@n1, B=0.6))) -> (Output_Get:Vector)
   > Set StoneRot_14(StoneRot_14:Rotator=MakeRotator(Pitch=float * float(A=@n1, B=8))) -> (Output_Get:Rotator)
   > Set Stone_15_Float(Stone_15_Float:Vector=MakeVector(X=float * float(A=@n1, B=0.8), Y=float * float(A=@n1, B=0.5), Z=float * float(A=@n1, B=-0.8))) -> (Output_Get:Vector)
   TryGetPawnOwner()
  ]}
]
variables [
 Stone_01_Float {type=Vector}
 Stone_02_Float {type=Vector}
 Stone_03_Float {type=Vector}
 Stone_04_Float {type=Vector}
 Stone_05_Float {type=Vector}
 Stone_06_Float {type=Vector}
 Stone_07_Float {type=Vector}
 Stone_08_Float {type=Vector}
 Stone_09_Float {type=Vector}
 Stone_10_Float {type=Vector}
 Stone_11_Float {type=Vector}
 Stone_12_Float {type=Vector}
 Stone_13_Float {type=Vector}
 Stone_14_Float {type=Vector}
 Stone_15_Float {type=Vector}
 StoneRot_01 {type=Rotator}
 StoneRot_04 {type=Rotator}
 StoneRot_05 {type=Rotator}
 StoneRot_06 {type=Rotator}
 StoneRot_10 {type=Rotator}
 StoneRot_12 {type=Rotator}
 StoneRot_14 {type=Rotator}
]
